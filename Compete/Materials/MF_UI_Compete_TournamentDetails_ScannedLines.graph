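# Material function for the Compete tournament details UI that outputs an animated horizontal scan-line mask.
# It pans input UVs vertically over time at the Speed input, then multiplies the V channel by LinesAmount and takes the fraction to repeat it.
# A very narrow SmoothStep just below Thickness turns that gradient into near hard-edged stripes.
# Each input falls back to a preview default, using scalar parameters for speed, amount and thickness.
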MaterialFunction /Compete/Materials/MF_UI_Compete_TournamentDetails_ScannedLines
graph {nodes [
  n1 = FunctionInput(Preview=ScalarParameter() {defaultValue=0.5, parameterName=ScanLinesThickness}) {bUsePreviewValueAsDefault=true, inputName=Thickness, inputType=FunctionInput_Scalar, sortPriority=3}
  n2 = Subtract(A=@n1) {constB=0.001}
  n3 = NamedRerouteUsage()
  n4 = Multiply(
   A=Time()
   B=FunctionInput(Preview=ScalarParameter() {defaultValue=0.02, parameterName=ScanLinesSpeed}) {bUsePreviewValueAsDefault=true, inputName=Speed, inputType=FunctionInput_Scalar, sortPriority=1}
  )
  n5 = Multiply(
   A=ComponentMask(None=Panner(Coordinate=@n3, Time=@n4) {speedY=0.5}) {g=true}
   B=FunctionInput(Preview=ScalarParameter() {defaultValue=50, parameterName=ScanLinesAmount}) {bUsePreviewValueAsDefault=true, inputName=LinesAmount, inputType=FunctionInput_Scalar, sortPriority=2}
  )
  FunctionOutput(None=SmoothStep(Min=Saturate(None=@n2), Max=@n1, Value=Frac(None=@n5))) {outputName=ScanLinesMask}
  NamedRerouteDeclaration(None=FunctionInput(Preview=TextureCoordinate()) {bUsePreviewValueAsDefault=true, inputName=UV, inputType=FunctionInput_Vector2}) {name=UV, nodeColor=<rgba 0 0 0 1>}
 ]}
