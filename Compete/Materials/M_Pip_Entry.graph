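# M_Pip_Entry is a translucent UI material for a pill-shaped pip with a progress fill and a Color parameter.
# A hard-edged horizontal mask comes from comparing normalized UI U against the saturated Progress parameter, using subtract, saturate, ceil and invert.
# MF_UI_Box draws the fully rounded shape, whose fill is scaled between half and full opacity by a mask times Progress Opacity, forming Pill Opacity.
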Material /Compete/Materials/M_Pip_Entry
graph {
 properties {blendMode=BLEND_Translucent, materialDomain=MD_UI, pixelDepthOffsetMode=PDOM_Legacy, usedShadingModels=Default Lit}
 nodes [
  n1 = MaterialFunctionCall() {materialFunction=&GetUserInterfaceUV}
  n2 = ScalarParameter() {defaultValue=0.25, parameterName=Progress, sliderMax=1}
  n3 = Saturate(None=Subtract(A=ComponentMask(None=@n1.Normalized UV) {r=true}, B=Saturate(None=@n2)))
  NamedRerouteDeclaration(None=OneMinus(None=Ceil(None=@n3))) {name=Progress Bar Mask, nodeColor=<rgba 0 0 0 1>}
  n4 = MaterialFunctionCall() {materialFunction=&GetUserInterfaceUV}
  n5 = Constant() {r=1}
  n6 = Constant() {r=1}
  n7 = StaticBool() {value=true}
  n8 = Constant() {r=0.5}
  n9 = Constant() {r=1}
  n10 = Multiply(A=NamedRerouteUsage(), B=ScalarParameter() {defaultValue=1, parameterName=Progress Opacity, sliderMax=1}) {constB=0}
  NamedRerouteDeclaration(None=Multiply(
    A=MaterialFunctionCall(UVs=@n4.Normalized UV, Size=@n5, CornerRoundness=@n6, KeepAspectRatio=@n7) {materialFunction=&MF_UI_Box}.Fill
    B=LinearInterpolate(A=@n8, B=@n9, Alpha=@n10)
   )) {name=Pill Opacity, nodeColor=<rgba 0 0 0 1>}
  StaticBool()
  VectorParameter() {parameterName=Color, defaultValue=<rgba 1 1 1 0>}
 ]
}
