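# Translucent UI font material for tournament prize text, drawing an outlined glyph with a gradient fade, judging by its name and parameters.
# Font signed distance and pixel distance factor, each scaled by the outline offset, multiply into a mask that drives a three-colour lerp from outline to body colour.
# The same mask, raised to a density power, blends start, outline and body alphas, which a rotatable linear gradient over 9-slice UVs multiplies.
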Material /Compete/Materials/M_TournamentPrizeTextOutline
graph {
 properties {bAutomaticallySetUsageInEditor=false, blendMode=BLEND_Translucent, materialDomain=MD_UI, usedShadingModels=Default Lit}
 nodes [
  n1 = FontSignedDistance()
  n2 = ScalarParameter() {defaultValue=0.2, group=Outline, parameterName=OutlineoOffset}
  n3 = Multiply(A=Multiply(A=@n1.Signed Distance, B=@n2), B=Multiply(A=@n1.Pixel Distance Factor, B=@n2))
  n4 = VectorParameter() {parameterName=BodyClolor, defaultValue=<rgba 1 1 1 1>}
  n5 = VectorParameter() {parameterName=OutlineClolor, defaultValue=<rgba 1 1 1 1>}
  MaterialFunctionCall(A=@n5.RGB, B=@n5.RGB, C=@n4.RGB, Alpha=@n3) {materialFunction=&Lerp_3Color}
  n6 = ScalarParameter() {defaultValue=1, group=Outline, parameterName=Dencity}
  Multiply(
   A=MaterialFunctionCall(
    UVs=MaterialFunctionCall() {materialFunction=&GetUserInterfaceUV}.9-Slice UV
    GradientStart=ScalarParameter() {group=Gradient, parameterName=Start}
    GradientEnd=ScalarParameter() {defaultValue=1, group=Gradient, parameterName=End}
    RotationAngle=ScalarParameter() {group=Gradient, parameterName=Rotation}
    RotationCenter=VectorParameter() {group=Gradient, parameterName=Center, defaultValue=<rgba 0.5 0.5 0 1>}.RGB
    GradientOffset=ScalarParameter() {group=Gradient, parameterName=Offset}
   ) {materialFunction=&MF_UI_LinearGradient}.Result
   B=MaterialFunctionCall(
    A=ScalarParameter() {group=Outline, parameterName=TextAlphaStart}
    B=ComponentMask(None=@n5.A) {a=true}
    C=ComponentMask(None=@n4.A) {a=true}
    Alpha=Power(Base=@n3, Exp=@n6)
   ) {materialFunction=&Lerp_3Color}.Result
  )
 ]
}
