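# Translucent UI material for the background of a Compete leaderboard tile, styled by button state.
# Modular button functions blend idle, hovered, focused, pressed and disabled colors and opacities, then their selected variants.
# UseMultiplyOverlay blends the base color toward that color multiplied by the Slate post buffer sample, before selection is applied.
# Resulting opacity is multiplied by a rounded-corner mask driven by the Roundness vector; the CornerRoundness scalar is unconnected.
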Material /Compete/Materials/M_UI_CompeteLeaderboardTileBackground
graph {
 properties {bAutomaticallySetUsageInEditor=false, blendMode=BLEND_Translucent, materialDomain=MD_UI, usedShadingModels=Default Lit}
 nodes [
  n1 = MaterialFunctionCall!(Idle, Idle Opacity, Hovered, Hovered Opacity, Focused, Focused Opacity, Pressed, Pressed Opacity, Disabled, Disabled Opacity) {materialFunction=/FortUILibrary/Systems/ModularButton/MaterialFunctions/MF_UIKit_ButtonStatesBase_Lerp}
  n2 = Multiply(A=MaterialFunctionCall!("UVs (Optional)") {materialFunction=/Engine/Functions/UserInterface/GetSlatePost0}.RGB, B=@n1.Color)
  n3 = ScalarParameter() {parameterName=UseMultiplyOverlay}
  n4 = VectorParameter() {parameterName=Roundness, defaultValue=<rgba 12 12 12 12>}
  n5 = AppendVector(A=@n4.RGB, B=@n4.RGB)
  n6 = ScalarParameter() {defaultValue=1}
  Multiply(
   A=MaterialFunctionCall!(
    Base States=LinearInterpolate(A=@n1.Color, B=@n2, Alpha=@n3)
    Base States Opacity=@n1.Opacity
    Selected_Idle
    Selected_Idle Opacity
    Selected_Hovered
    Selected_Hovered Opacity
    Selected_Focused
    Selected_Focused Opacity
    Selected_Pressed
    Selected_Pressed Opacity
    Selected_Disabled
    Selected_Disabled Opacity
   ) {materialFunction=/FortUILibrary/Systems/ModularButton/MaterialFunctions/MF_UIKit_ButtonStatesSelectable_Lerp}.Opacity
   B=Multiply(A=MaterialFunctionCall(Corner Radius=@n5, Radius in Pixels=@n6) {materialFunction=&MF_UI_CoreUI_RoundedMask}.Shape Mask)
  )
  ScalarParameter() {defaultValue=12, parameterName=CornerRoundness}
 ]
}
