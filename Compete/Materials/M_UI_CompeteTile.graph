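# Translucent UI material for a Compete tile that shows a character image over a colored gradient with rounded corners.
# A composite subgraph scales fill-fitted UVs around a horizontally centered, upper pivot, driven by zoom percent and a Hover parameter.
# A sharpened reflected gradient blends the Character texture into GradientColor.
# The shared Item Shop rounded-corner function supplies the corner shape, with tile-driven radius override.
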Material /Compete/Materials/M_UI_CompeteTile
graph {
 properties {bAutomaticallySetUsageInEditor=false, blendMode=BLEND_Translucent, materialDomain=MD_UI, usedShadingModels=Default Lit}
 nodes [
  n1 = ScalarParameter() {defaultValue=25, group=Image, parameterName=ZoomImage_Percent, sliderMax=100, sliderMin=-100}
  n2 = MaterialFunctionCall(UV=MaterialFunctionCall() {materialFunction=&GetUserInterfaceUV}.Normalized UV) {materialFunction=&MF_UI_UVScaling}
  Composite!(is vertical ratio, Zoom image percent=@n1, UVs=@n2.Scaled To Fill) {subgraphName=Handle Image Zoom}
  n3 = MaterialFunctionCall(Float2=TextureCoordinate()) {materialFunction=&BreakOutFloat2Components}
  n4 = MaterialFunctionCall!(
   UVs=MaterialFunctionCall(X=@n3.R, Y=Add(A=@n3.G) {constB=0.064}) {materialFunction=&MakeFloat2}.Result
   Offset=Constant() {r=0.05}
   RotationCenter
   RotationAngle=Constant() {r=0.25}
  ) {materialFunction=/FortUILibrary/MaterialFunctions/Gradients/MF_UI_ReflectedGradient}
  n5 = Saturate(None=Power(Base=Multiply(A=@n4.Result) {constB=1.4}) {constExponent=1.5})
  LinearInterpolate(
   A=Multiply(A=OneMinus(None=@n5), B=VectorParameter() {parameterName=GradientColor, defaultValue=<rgba 0.003 0.1248 0.3916 1>}.RGB)
   B=TextureSampleParameter2D(UVs=NamedRerouteUsage()) {parameterName=Character, texture=/Compete/Textures/T_UI_Compete_Placeholder}.RGB
   Alpha=@n5
  )
  n6 = VectorParameter() {group=Tile-driven params, parameterName=CornerRadius}
  MaterialFunctionCall!(
   Input Alpha=Constant() {r=1}
   bOverrideCornerRadius=ScalarParameter() {group=Tile-driven params, parameterName=IsOverrideCornerRadius}
   Corner Radius Override=AppendVector(A=@n6.RGB, B=@n6.RGB)
  ) {materialFunction=/FortUILibrary/Systems/ItemShopTile/Materials/MaterialFunctions/MF_UI_ItemShop_RoundedCorners}
  MaterialExpressionReroute_3 = Reroute(None=@n2.Scaled To Fill)
  n7 = ScalarParameter() {defaultValue=50, group=Animation, parameterName=Hover_Zoom_Increment}
  MaterialExpressionReroute_2 = Reroute(None=@n1)
  n8 = Multiply(A=LinearInterpolate(
    A=@MaterialExpressionReroute_2.Zoom image percent
    B=Add(A=@MaterialExpressionReroute_2.Zoom image percent, B=@n7) {constB=10}
    Alpha=ScalarParameter() {group=Animation, parameterName=Hover}
   ) {constA=0.95}) {constB=0.01}
  n9 = Divide(A=Constant() {r=344}, B=Constant() {r=720})
  n10 = Multiply(A=LinearInterpolate(B=Subtract(B=@n7) {constA=100}, Alpha=LinearInterpolate(B=@n9, Alpha=@n8) {constB=0.325}) {constAlpha=100, constB=95}) {constB=0.01}
  n11 = Clamp(None=LinearInterpolate(Alpha=@n10) {constA=1, constB=2}) {maxDefault=2}
  n12 = AppendVector(A=Constant() {r=0.5}, B=LinearInterpolate(A=Constant(), B=Constant() {r=0.125}, Alpha=@n10))
  n13 = Subtract(
   A=Add(A=Divide(A=@MaterialExpressionReroute_3.UVs, B=@n11), B=@n12)
   B=Divide(A=@n12, B=@n11) {desc=Scale the center offset by the scaling factor}
  ) {desc=Re offset to maintain center}
  MaterialExpressionReroute_4 = Reroute(None=@n13)
  NamedRerouteDeclaration(None=@MaterialExpressionReroute_4.UVs) {name=ImageUVs, nodeColor=<rgba 1 0.6562 0 1>}
  PinBase() {reroutePins [
    is vertical ratio {expression=&self:MaterialExpressionReroute_314}
    Zoom image percent {expression=&self:MaterialExpressionReroute_2}
    UVs {expression=&self:MaterialExpressionReroute_3}
   ]}
  PinBase!(UVs=@n13) {reroutePins [UVs {expression=&self:MaterialExpressionReroute_4}]}
  MaterialExpressionReroute_314 = Reroute()
 ]
}
