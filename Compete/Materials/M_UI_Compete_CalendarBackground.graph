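# UI material that draws the Compete calendar background: a gradient backdrop with an animatable colored ellipse glow and an optional texture.
# A rotated reflected gradient forms the base, overlaid by aspect-corrected radial-gradient ellipses whose position, fade and colors follow transition parameters.
# The texture is corrected for its aspect ratio, scaled around an anchor, masked by a radial gradient and toggled by IsShowTexture.
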Material /Compete/Materials/M_UI_Compete_CalendarBackground
graph {
 properties {bAutomaticallySetUsageInEditor=false, materialDomain=MD_UI, usedShadingModels=Default Lit}
 nodes [
  Composite() {subgraphName=Fix widget aspect ratio}
  Composite() {subgraphName=Fix widget aspect ratio}
  n1 = ScalarParameter() {defaultValue=1, group=Texture, parameterName=TextureScale}
  n2 = AppendVector(A=ScalarParameter() {defaultValue=1, group=Texture, parameterName=Anchor X}, B=ScalarParameter() {group=Texture, parameterName=Anchor Y})
  MaterialExpressionReroute_0 = Reroute(None=@n2)
  n3 = TextureObjectParameter() {parameterName=Texture, texture=/Compete/Textures/T_UI_Compete_Placeholder}
  n4 = TextureProperty(Tex=@n3) {property=TMTM_TexelSize}
  n5 = AppendVector(A=Constant() {r=1}, B=Divide(A=ComponentMask(None=@n4) {r=true}, B=ComponentMask(None=@n4) {g=true})) {desc=Texture Vertical Aspect Ratio}
  n6 = Multiply(
   A=Subtract(A=TextureCoordinate(), B=@n2)
   B=Divide(A=Divide(A=@MaterialExpressionReroute_0.Output, B=@n5), B=ScalarParameter() {defaultValue=0.75, group=Texture, parameterName=UserScale})
  ) {desc="Scaling to fix Aspect Ratio, fit to width"}
  n7 = MaterialFunctionCall!(
   Texture Scale=AppendVector(A=@n1, B=@n1)
   UVs=NamedRerouteDeclaration(None=Add(A=@n6, B=@MaterialExpressionReroute_0)) {name=AdjustedTextureUV, nodeColor=<rgba 0.8984 1 0 1>}
   Offset XY Texturesize Z=Constant3Vector() {constant=<rgba 1 0 1 0>}
  ) {materialFunction=/Engine/Functions/Engine_MaterialFunctions01/Texturing/ScaleUVsAroundPoint}
  n8 = ScalarParameter() {group=Texture, parameterName=GradientOffset_X}
  n9 = ScalarParameter() {group=Texture, parameterName=GradientOffset_Y}
  n10 = MaterialFunctionCall(
   UVs=Subtract(A=NamedRerouteUsage(), B=AppendVector(A=@n8, B=@n9))
   Texture Scale=AppendVector(
    A=ScalarParameter() {defaultValue=1, group=Texture, parameterName=GradientScale_X}
    B=ScalarParameter() {defaultValue=1, group=Texture, parameterName=GradientScale_Y}
   )
  ) {materialFunction=&ScaleUVsByCenter}
  n11 = ScalarParameter() {defaultValue=1, group=Texture, parameterName=RadialGradientRadius}
  n12 = ScalarParameter() {defaultValue=2.33, group=Texture, parameterName=TextureGradientDensity}
  LinearInterpolate(
   A=NamedRerouteUsage()
   B=LinearInterpolate(
    A=NamedRerouteUsage()
    B=TextureSample(UVs=@n7.UVs, Tex=@n3) {samplerSource=SSM_Clamp_WorldGroupSettings}.RGB
    Alpha=MaterialFunctionCall(UVs=@n10.UVs, Radius=@n11, Density=@n12) {materialFunction=&RadialGradientExponential}.RadialGradientExponential
   )
   Alpha=ScalarParameter() {parameterName=IsShowTexture}
  ) {constAlpha=1}
  n13 = MaterialFunctionCall() {materialFunction=&GetUserInterfaceUV}
  n14 = AppendVector(A=Constant() {r=1}, B=Divide(A=ComponentMask(None=@n13.Pixel Size) {g=true}, B=ComponentMask(None=@n13.Pixel Size) {r=true})) {desc=Widget Vertical Aspect Ratio}
  MaterialExpressionReroute_1 = Reroute(None=@n14)
  n15 = Constant() {r=1}
  n16 = Multiply(A=TextureCoordinate(), B=Divide(A=@MaterialExpressionReroute_1.Output, B=AppendVector(A=@n15, B=@n15) {desc=Texture Vertical Aspect Ratio}))
  n17 = Divide(B=ScalarParameter() {defaultValue=1.8769, group=Elipse, parameterName=Scale_X})
  n18 = Multiply(A=ComponentMask(None=@n16) {r=true}, B=@n17) {constB=0.5}
  n19 = Divide(A=@n17)
  n20 = ScalarParameter() {defaultValue=0.382, group=Elipse, parameterName=OffsetX}
  n21 = Divide(B=ScalarParameter() {defaultValue=0.776, group=Elipse, parameterName=Scale_Y})
  n22 = Multiply(A=ComponentMask(None=@n16) {g=true}, B=@n21) {constB=0.5}
  n23 = Divide(A=@n21)
  n24 = Clamp(None=ScalarParameter() {group=Elipse, parameterName=AnimElipseTransition})
  n25 = Add(
   A=Multiply(A=Multiply(A=@n24) {constB=-1}, B=ScalarParameter() {defaultValue=1, group=Elipse, parameterName=TransitionDistance})
   B=ScalarParameter() {defaultValue=-0.352, group=Elipse, parameterName=OffsetY}
  )
  n26 = AppendVector(
   A=Subtract(A=Subtract(A=@n18, B=@n19), B=Multiply(A=@n17, B=@n20)) {constB=0}
   B=Subtract(A=Subtract(A=@n22, B=@n23), B=Multiply(A=@n21, B=@n25)) {constB=0}
  )
  n27 = MaterialFunctionCall(
   UVs=@n26
   CenterPosition=Constant()
   Radius=ScalarParameter() {defaultValue=0.3272, group=Elipse, parameterName=MRadius}
   Density=ScalarParameter() {defaultValue=1, group=Elipse, parameterName=MDensity}
  ) {materialFunction=&RadialGradientExponential}
  n28 = MaterialFunctionCall(
   UVs=@n26
   CenterPosition=Constant()
   Radius=ScalarParameter() {defaultValue=0.4775, group=Elipse, parameterName=Radius}
   Density=ScalarParameter() {defaultValue=3, group=Elipse, parameterName=Density}
  ) {materialFunction=&RadialGradientExponential}
  LinearInterpolate(A=@n27.RadialGradientExponential, B=@n28.RadialGradientExponential, Alpha=@n28.RadialGradientExponential)
  n29 = ScalarParameter() {parameterName=AnimColorTransition}
  n30 = LinearInterpolate(
   A=VectorParameter() {group=Elipse, parameterName=Color2, defaultValue=<rgba 0.0423 0.0027 0.0212 1>}.RGB
   B=VectorParameter() {group=Elipse, parameterName=AnimColor2, defaultValue=<rgba 0 0.0003 0.0423 1>}.RGB
   Alpha=@n29
  )
  n31 = LinearInterpolate(
   A=VectorParameter() {group=Elipse, parameterName=Color1, defaultValue=<rgba 0.855 0.0056 0.0242 1>}.RGB
   B=VectorParameter() {group=Elipse, parameterName=AnimColor1, defaultValue=<rgba 0.0444 0.0488 0.6562 1>}.RGB
   Alpha=@n29
  )
  n32 = ScalarParameter() {group=Elipse, parameterName=AnimElipseTransition}
  NamedRerouteDeclaration(None=LinearInterpolate(
    A=NamedRerouteUsage()
    B=LinearInterpolate(A=@n30, B=@n31, Alpha=@n27.RadialGradientExponential)
    Alpha=LinearInterpolate(A=@n28.RadialGradientExponential, Alpha=@n32) {constB=0}
   ) {constA=1}) {name=Iteration2BG, nodeColor=<rgba 0 0 0 1>}
  n33 = ScalarParameter() {defaultValue=-0.2, group=Background, parameterName=Offset}
  n34 = ScalarParameter() {defaultValue=0.05, group=Background, parameterName=RotationAngle}
  NamedRerouteDeclaration(None=LinearInterpolate(
    A=VectorParameter() {group=Background, parameterName=MainColor, defaultValue=<rgba 0 0.0118 0.0627 1>}.RGB
    B=VectorParameter() {group=Background, parameterName=SecondColor, defaultValue=<rgba 0.004 0.003 0.0152 1>}.RGB
    Alpha=MaterialFunctionCall!(UVs, Offset=@n33, RotationCenter, RotationAngle=@n34) {materialFunction=/FortUILibrary/MaterialFunctions/Gradients/MF_UI_ReflectedGradient}.Result
   )) {name=Background, nodeColor=<rgba 0 0 0 1>}
  PinBase()
  n35 = MaterialFunctionCall() {materialFunction=&GetUserInterfaceUV}
  n36 = ComponentMask(None=@n35.Pixel Size) {g=true}
  n37 = ComponentMask(None=@n35.Pixel Size) {r=true}
  PinBase!(Output=AppendVector(A=Constant() {r=1}, B=Divide(A=@n36, B=@n37)) {desc=Widget Vertical Aspect Ratio}) {reroutePins [Output {expression=&self:MaterialGraph_0.MaterialGraphNode_Composite_1.Fix widget aspect ratio.MaterialExpressionReroute_0}]}
  PinBase!()
  PinBase!(Output=@n14) {reroutePins [Output {expression=&self:MaterialExpressionReroute_1}]}
 ]
}
