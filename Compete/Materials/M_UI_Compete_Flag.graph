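# Translucent UI material for the Compete flag, a rounded shape with a triangular cutout.
# A rounded-rectangle mask, driven by the Corner Radius parameter, is multiplied by the Color parameter's alpha.
# An antialiased triangle SDF on vertically offset UVs then lerps that mask to zero, carving the notch procedurally without textures.
# A widget aspect-ratio calculation is present but feeds nothing.
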Material /Compete/Materials/M_UI_Compete_Flag
graph {
 properties {bAutomaticallySetUsageInEditor=false, blendMode=BLEND_Translucent, materialDomain=MD_UI, usedShadingModels=Default Lit}
 nodes [
  n1 = MaterialFunctionCall() {materialFunction=&GetUserInterfaceUV}
  AppendVector(A=Constant() {r=1}, B=Divide(A=ComponentMask(None=@n1.Pixel Size) {g=true}, B=ComponentMask(None=@n1.Pixel Size) {r=true})) {desc=Widget Vertical Aspect Ratio}
  n2 = VectorParameter() {group=Corners, parameterName=Corner Radius, sortPriority=2, defaultValue=<rgba 0.1 0 0.1 0>}
  n3 = AppendVector(A=@n2.RGB, B=@n2.RGB)
  n4 = Constant()
  n5 = MaterialFunctionCall(Float2=TextureCoordinate()) {materialFunction=&BreakOutFloat2Components}
  n6 = MaterialFunctionCall(X=@n5.R, Y=Subtract(A=@n5.G) {constB=0.05}) {materialFunction=&MakeFloat2}
  n7 = Constant2Vector() {g=1, r=-1.3}
  n8 = Constant2Vector() {g=0.5}
  n9 = Constant2Vector() {g=1, r=1.3}
  LinearInterpolate(
   A=Multiply(
    A=MaterialFunctionCall(Corner Radius=@n3, Radius in Pixels=@n4) {materialFunction=&MF_UI_CoreUI_RoundedMask}.Shape Mask
    B=VectorParameter() {parameterName=Color, defaultValue=<rgba 0.0395 1 0.2307 1>}.A
   )
   Alpha=MaterialFunctionCall(DistanceField=MaterialFunctionCall!(UVs=@n6.Result, Corner 01=@n7, Corner 02=@n8, Corner 03=@n9, Glow Min, Glow Max, Outline Thickness) {materialFunction=/FortUILibrary/MaterialFunctions/Shapes/MF_UI_WIP_Triangle}.SDF) {materialFunction=&MF_UI_AntiAliasing}.Inner
  ) {constB=0}
 ]
}
