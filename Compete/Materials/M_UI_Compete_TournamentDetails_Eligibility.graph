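# Translucent UI material for the tournament details eligibility display in Compete, built around a locked-padlock MSDF icon.
# The icon texture is scaled about its center, converted from MSDF to SDF and antialiased into IconMask, alongside an IconColor parameter.
# A looping sine, timed by AnimationSpeed, feeds a spring curve that drives a viewport-scaled vertical bob in IconMovement.
# The graph also holds a second, similar set of offset and MSDF sampling nodes.
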Material /Compete/Materials/M_UI_Compete_TournamentDetails_Eligibility
graph {
 properties {bAutomaticallySetUsageInEditor=false, blendMode=BLEND_Translucent, materialDomain=MD_UI, usedShadingModels=Default Lit}
 nodes [
  n1 = Constant()
  n2 = ConstantBiasScale(None=Sine(None=NamedRerouteUsage()))
  n3 = ScalarParameter() {defaultValue=0.1, parameterName=Stiffness}
  n4 = ScalarParameter() {parameterName=Damping}
  n5 = LinearInterpolate(Alpha=OneMinus(None=MaterialFunctionCall!("Value (0-1)"=@n2, Stiffness=@n3, Damping=@n4) {materialFunction=/FortUILibrary/MaterialFunctions/Basics/MF_UI_SpringCurve}.Spring Curve)) {constA=5, constB=-5}
  Add(
   A=WorldPosition().XY
   B=Multiply(
    A=AppendVector(A=@n1, B=@n5)
    B=MaterialFunctionCall() {materialFunction=/FortUILibrary/MaterialFunctions/Basics/MF_UI_ViewportScale}.Viewport Scale Factor
   )
  )
  Constant() {r=1}
  n6 = Add(A=MaterialFunctionCall() {materialFunction=&GetUserInterfaceUV}.Normalized UV, B=Constant2Vector() {g=0.01})
  n7 = NamedRerouteDeclaration(None=Rotator(
    Coordinate=MaterialFunctionCall(UV=@n6) {materialFunction=&MF_UI_UVScaling}.Scaled to Fit
    Time=Constant()
   ) {speed=6.2832}) {name=TextureUV, nodeColor=<rgba 0 0 0 1>}
  n8 = NamedRerouteDeclaration(None=TextureObject() {samplerType=SAMPLERTYPE_LinearColor, texture=/FortUILibrary/Textures/Icons/T_UI_Icon_UIKit_Locked_MSDF}) {name=TextureObject, nodeColor=<rgba 0 0 0 1>}
  MaterialFunctionCall(
   DistanceField=MaterialFunctionCall!(MSDF RGB=TextureSample(UVs=@n7, Tex=@n8).RGB) {materialFunction=/FortUILibrary/MaterialFunctions/Shapes/Modifiers/MF_UI_MSDFtoSDF}.SDF
   Threshold=Constant() {r=0.5}
  ) {materialFunction=&MF_UI_AntiAliasing}
  n9 = Constant2Vector() {g=30, r=30}
  n10 = Frac(None=Multiply(A=Time(), B=ScalarParameter() {defaultValue=1, parameterName=AnimationSpeed}))
  n11 = MaterialFunctionCall(
   "Value (0-1)"=ConstantBiasScale(None=Sine(None=@n10))
   Stiffness=ScalarParameter() {defaultValue=0.1, parameterName=Stiffness}
   Damping=ScalarParameter() {parameterName=Damping}
  ) {materialFunction=&MF_UI_SpringCurve}
  n12 = AppendVector(A=Constant(), B=LinearInterpolate(Alpha=OneMinus(None=@n11.Spring Curve)) {constA=-5, constB=5})
  n13 = MaterialFunctionCall() {materialFunction=&MF_UI_ViewportScale}
  NamedRerouteDeclaration(None=Add(
    A=MaterialFunctionCall!(Anchor, Translation, Padding=@n9, Scale) {materialFunction=/FortUILibrary/MaterialFunctions/Basics/MF_UI_VertexTransform}.Transformed Position
    B=Multiply(A=@n12, B=@n13.Viewport Scale Factor)
   )) {name=IconMovement, nodeColor=<rgba 0 0 0 1>}
  n14 = TextureSample(UVs=MaterialFunctionCall(UVs=TextureCoordinate(), Texture Scale=ScalarParameter() {defaultValue=0.7, parameterName=TextureScale}) {materialFunction=&ScaleUVsByCenter}.UVs) {
   samplerSource=SSM_Clamp_WorldGroupSettings
   samplerType=SAMPLERTYPE_LinearColor
   texture=&T_UI_Icon_UIKit_Locked_MSDF
  }
  NamedRerouteDeclaration(None=MaterialFunctionCall(
    DistanceField=MaterialFunctionCall(MSDF RGB=@n14.RGB) {materialFunction=&MF_UI_MSDFtoSDF}.SDF
    Threshold=ScalarParameter() {defaultValue=0.52, parameterName=Threshold}
   ) {materialFunction=&MF_UI_AntiAliasing}.Outer) {name=IconMask, nodeColor=<rgba 0 0 0 1>}
  n15 = Time()
  n16 = ScalarParameter() {defaultValue=1, parameterName=AnimationSpeed}
  NamedRerouteDeclaration(None=Frac(None=Multiply(A=@n15, B=@n16))) {name=TextureTime, nodeColor=<rgba 0 0 0 1>}
  VectorParameter() {parameterName=IconColor, defaultValue=<rgba 0.0048 0.004 0.0116 1>}
 ]
}
