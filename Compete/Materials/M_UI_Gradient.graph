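# M_UI_Gradient is a translucent UI-domain material that builds a linear gradient confined to a rounded-corner shape.
# MF_UI_LinearGradient takes start, end, angle and offset parameters, and its result is multiplied by the shape mask from MF_UI_CoreUI_RoundedMask.
# The mask takes a per-corner radius vector and a Radius in Pixels parameter.
# A separate Color parameter defaults to black; the graph shows no connections to material outputs.
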Material /Compete/Materials/M_UI_Gradient
graph {
 properties {bAutomaticallySetUsageInEditor=false, blendMode=BLEND_Translucent, materialDomain=MD_UI, usedShadingModels=Default Lit}
 nodes [
  n1 = VectorParameter() {
   group=Shape
   parameterName=Corner Radius
   sortPriority=1
   channelNames {a=BottomRight, b=TopRight, g=BottomLeft, r=TopLeft}
   defaultValue=<rgba 10 10 10 10>
  }
  Multiply(
   A=MaterialFunctionCall(
    GradientStart=ScalarParameter() {defaultValue=-0.2, parameterName=Start}
    GradientEnd=ScalarParameter() {defaultValue=0.5, parameterName=End}
    RotationAngle=ScalarParameter() {defaultValue=0.25, parameterName=Angle}
    GradientOffset=ScalarParameter() {defaultValue=0.2, parameterName=GradientOffset}
   ) {materialFunction=&MF_UI_LinearGradient}.Result
   B=MaterialFunctionCall(
    Corner Radius=AppendVector(A=@n1, B=@n1)
    Radius in Pixels=ScalarParameter() {defaultValue=1, group=Shape, parameterName=Radius in Pixels, sliderMax=1, sortPriority=2}
   ) {materialFunction=&MF_UI_CoreUI_RoundedMask}.Shape Mask
  )
  VectorParameter() {parameterName=Color, defaultValue=<rgba 0 0 0 1>}
 ]
}
