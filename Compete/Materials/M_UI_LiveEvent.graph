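# UI material for a live event widget, drawing an animated, colored glow with rounded corners and alpha-composite blending.
# Panned blurred noise in radial coordinates forms moving streaks, sampled with explicit derivatives from the widget UV.
# Exponential radial gradients on aspect-corrected, anchor-scaled UVs fade the streaks and shift the tint between the Color parameter and white.
# The streak value and a rounded-rectangle mask premultiply the color.
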Material /Compete/Materials/M_UI_LiveEvent
graph {
 properties {bAutomaticallySetUsageInEditor=false, blendMode=BLEND_AlphaComposite, materialDomain=MD_UI, usedShadingModels=Default Lit}
 nodes [
  n1 = VectorParameter() {parameterName=Color, defaultValue=<rgba 1 0.0097 0 1>}
  n2 = MaterialFunctionCall(UVs=NamedRerouteUsage(), CenterPosition=Constant2Vector(), Radius=Constant() {r=1.32}, Density=Constant() {r=0.5}) {materialFunction=&RadialGradientExponential}
  n3 = Multiply(A=MaterialFunctionCall() {materialFunction=&GetUserInterfaceUV}.Tiling) {constB=0.1}
  n4 = MaterialFunctionCall(Float2=MaterialFunctionCall!(Vector or UVs=ConstantBiasScale(None=@n3) {bias=0.5, scale=2}) {materialFunction=/FortUILibrary/MaterialFunctions/Basics/MF_UI_CheaperVectorToRadial}.Radial Coordinates) {materialFunction=&BreakOutFloat2Components}
  n5 = Multiply(A=@n4.R, B=ScalarParameter() {defaultValue=1, parameterName=Perpendicular Repitition})
  n6 = Multiply(A=@n4.G, B=ScalarParameter() {defaultValue=2, parameterName=Parallel Repitition})
  n7 = ScalarParameter() {defaultValue=-0.05, parameterName=Perpendicular Speed}
  n8 = ScalarParameter() {defaultValue=-0.1, parameterName=Parallel Speed}
  n9 = MaterialFunctionCall() {materialFunction=&GetUserInterfaceUV}
  n10 = TextureSample!(
   UVs=Panner(Coordinate=AppendVector(A=@n5, B=@n6), Time=Time(), Speed=AppendVector(A=@n7, B=@n8))
   Tex
   "DDX(UVs)"=DDX(Value=@n9.Normalized UV)
   "DDY(UVs)"=DDY(Value=@n9.Normalized UV)
   Apply View MipBias
  ) {mipValueMode=TMVM_Derivative, samplerType=SAMPLERTYPE_Grayscale, texture=/FortUILibrary/UIKit/Buttons/Textures/LowResBlurredNoise}
  n11 = NamedRerouteUsage()
  n12 = Constant2Vector()
  n13 = Constant() {r=1}
  n14 = Constant() {r=1.3}
  n15 = Multiply(
   A=@n10.R
   B=SmoothStep(Value=MaterialFunctionCall(UVs=@n11, CenterPosition=@n12, Radius=@n13, Density=@n14) {materialFunction=&RadialGradientExponential}.RadialGradientExponential)
  )
  n16 = Constant() {r=1}
  n17 = ScalarParameter() {defaultValue=16, parameterName=Corner Radius}
  Multiply(
   A=LinearInterpolate(B=LinearInterpolate(A=@n1.RGB, Alpha=@n2.RadialGradientExponential) {constA=1}, Alpha=@n15)
   B=Multiply(
    A=@n15
    B=MaterialFunctionCall(Size=@n16, Corner Radius=@n17, Radius in Pixels=@n16) {materialFunction=&MF_UI_CoreUI_RoundedMask}.Shape Mask
   )
  )
  n18 = ScalarParameter() {defaultValue=1, group=Texture, parameterName=TextureScale}
  n19 = AppendVector(A=ScalarParameter() {group=Texture, parameterName=Anchor X}, B=ScalarParameter() {group=Texture, parameterName=Anchor Y})
  n20 = MaterialFunctionCall() {materialFunction=&GetUserInterfaceUV}
  n21 = Divide(A=ComponentMask(None=@n20.Pixel Size) {r=true}, B=ComponentMask(None=@n20.Pixel Size) {g=true})
  n22 = Constant() {r=1}
  n23 = Multiply(
   A=Subtract(A=TextureCoordinate(), B=@n19)
   B=Divide(
    A=AppendVector(A=@n21, B=@n22) {desc=Widget horizontal Aspect Ratio}
    B=ScalarParameter() {defaultValue=1, group=Texture, parameterName=UserScale}
   )
  ) {desc="Scaling to fix Aspect Ratio, fit to width"}
  NamedRerouteDeclaration(None=MaterialFunctionCall!(Texture Scale=AppendVector(A=@n18, B=@n18), UVs=Add(A=@n23, B=@n19), Offset XY Texturesize Z=Constant3Vector() {constant=<rgba 1 0 1 0>}) {materialFunction=/Engine/Functions/Engine_MaterialFunctions01/Texturing/ScaleUVsAroundPoint}.UVs) {name=UVs, nodeColor=<rgba 0 1 0.0391 1>}
 ]
}
