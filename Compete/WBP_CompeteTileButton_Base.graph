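# Base Compete tile button widget, built on UIKitModularButton, that drives its modular blocks' visual states and has hover and press sound helpers.
# PreConstruct sets block transition timings and binds input-method changes.
# Button events request block transitions, with selection also passed to each block and optionally applied instantly through a one-shot flag.
# Gamepad or touch input makes selected tiles show as focused, while touch suppresses hover focus.
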WidgetBlueprint /Compete/WBP_CompeteTileButton_Base
graphs [
 EventGraph {nodes [
   n1 = OnInputTypeChanged!() -> (OutputDelegate, then, bNewInputType:ECommonInputType Enum) {is=title K2Node_CustomEvent}
   > Branch(Condition=AND Boolean(A=ShouldFocusOnSelected!() -> (ShouldFocus:bool), B=GetSelected!(self:Common Button Base Object Reference) -> (ReturnValue:bool)))
   > RequestTransitionOnBlocks!(execute, self:UIKit Modular Button Object Reference, State:EUIKitBlockVisualState Enum=Focused) -> (then)
   n2 = Event PreConstruct()
   > Parent: PreConstruct(IsDesignTime=@n2.IsDesignTime)
   > n3 = Set BlockTiming(BlockTiming:UIKit Block Timing Structure=Make UIKit Block Timing!(
     Hovering:float=0
     Unhovering:float=0
     Focusing:float=0.3
     Unfocusing:float=0.15
     Pressing:float=0.1
     Releasing:float=0.1
     Disabling:float=0.1
     Enabling:float=0.1
     Selecting:float=0.2
     DeselectingIdle:float=0.2
     DeselectingFocused:float=0.2
    ) -> (UIKitBlockTiming:UIKit Block Timing Structure) {is=title K2Node_MakeStruct}) -> (Output_Get:UIKit Block Timing Structure)
   > SetBlocksTiming!(execute, self:UIKit Modular Button Object Reference, Timing:"UIKit Block Timing Structure (by ref)"=@n3) -> (then)
   > Assign On Input Method Changed!(execute:Exec, self:Common Input Object Reference=Get CommonInputSubsystem(), Delegate:"Delegate (by ref)"=@n1) -> (then:Exec) {is=title K2Node_AssignDelegate}
   Event On Disabled() {is=title K2Node_Event}
   > RequestTransitionOnBlocks(State=Disabled)
   Event On Deselected()
   > Set Blocks Is Selected!(execute, self, Is Selected:bool=false) -> (then) {is=title K2Node_CallFunction}
   n4 = ShouldFocusOnSelected()
   n5 = IsHovered()
   > RequestTransitionOnBlocks(State=Select(Option 0:EUIKitBlockVisualState Enum=DeselectedIdle, Option 1:EUIKitBlockVisualState Enum=DeselectedFocused, Index=OR Boolean(A=@n5, B=@n4)) -> (ReturnValue:EUIKitBlockVisualState Enum))
   > Branch(Condition=ShouldFocusOnSelected())
   > RequestTransitionOnBlocks(State=Unfocused)
   Event On Hovered()
   n6 = Get CommonInputSubsystem()
   > n7 = Branch(Condition="Equal (Enum)"(A=GetCurrentInputType(self=@n6), B=Touch))
   RequestTransitionOnBlocks(execute=@n7.else, State=Focused)
   Event On Unhovered()
   > RequestTransitionOnBlocks(State=Unfocused)
   Event On Pressed() {is=title K2Node_Event}
   > RequestTransitionOnBlocks(State=Pressed)
   Event On Released() {is=title K2Node_Event}
   > RequestTransitionOnBlocks(State=Released)
   Event On Selected()
   > Set Blocks Is Selected(Is Selected=true)
   > n8 = Branch(Condition=Get IsInstantSelection() -> (IsInstantSelection:bool))
   > RequestInstantTransitionOnBlocks!(execute, self:UIKit Modular Button Object Reference, State:EUIKitBlockInstantTransitionState Enum=InstantSelected) -> (then)
   > Set IsInstantSelection(IsInstantSelection:bool=false) -> (Output_Get:bool)
   n9 = ShouldFocusOnSelected()
   n10 = IsHovered()
   RequestTransitionOnBlocks(execute=@n8.else, State=Select(Option 0=Selected, Option 1=Focused, Index=OR Boolean(A=@n10, B=@n9)))
   Event On Enabled() {is=title K2Node_Event}
   > RequestTransitionOnBlocks(State=Enabled)
  ]}
 GetBlockTimings {nodes [
   GetBlockTimings() {is=title K2Node_FunctionEntry}
   > Return Node!(execute, BlockTiming:UIKit Block Timing Structure=Get BlockTiming() -> (BlockTiming:UIKit Block Timing Structure))
  ]}
 PlaySoundHovered {nodes [
   PlaySoundHovered() {is=title K2Node_FunctionEntry}
   > n1 = Get!(execute:Exec) -> (then:Exec, else:Exec, HoveredSound:Sound Base Object Reference)
   > SpawnSound2D!(
    execute
    Sound:Sound Base Object Reference=@n1
    VolumeMultiplier:float=1
    PitchMultiplier:float=1
    StartTime:float=0
    ConcurrencySettings:Sound Concurrency Object Reference
    bPersistAcrossLevelTransition:bool=false
    bAutoDestroy:bool=true
   ) -> (then, ReturnValue:Audio Component Object Reference)
  ]}
 PlaySoundPressed {nodes [
   PlaySoundPressed() {is=title K2Node_FunctionEntry}
   > n1 = Get!(execute) -> (then, else, PressedSound:Sound Base Object Reference)
   > SpawnSound2D(Sound=@n1, VolumeMultiplier=1, PitchMultiplier=1, StartTime=0, bPersistAcrossLevelTransition=false, bAutoDestroy=true)
  ]}
 Set Blocks Is Selected {nodes [
   n1 = Set Blocks Is Selected!() -> (then, Is Selected:bool=~) {is=title K2Node_FunctionEntry}
   > n2 = For Each Loop(Array:Array of UIKit Block Object References=Get ModularBlocks() -> (ModularBlocks:Array of UIKit Block Object References)) -> (Array Element:UIKit Block Object Reference)
   > n3 = Cast To WBP_UIKit_Block_Base!(execute, Object=@n2) -> (then, CastFailed, AsWBP UIKit Block Base:WBP UIKit Block Base Object Reference) {is=title K2Node_DynamicCast}
   > SetIsSelected!(execute, self:WBP UIKit Block Base Object Reference=@n3, IsSelected:bool=@n1) -> (then)
  ]}
 ShouldFocusOnSelected {nodes [
   ShouldFocusOnSelected() {is=title K2Node_FunctionEntry}
   n1 = GetCurrentInputType(self=Get CommonInputSubsystem())
   > Return Node!(execute, ShouldFocus:bool=OR Boolean(A="Equal (Enum)"(A=@n1, B=Gamepad), B="Equal (Enum)"(A=@n1, B=Touch)))
  ]}
]
variables [
 BlockTiming {
  type=UIKitBlockTiming
  default {
   deselectingFocused=0
   deselectingIdle=0
   disabling=0
   enabling=0
   focusing=0
   hovering=0
   pressing=0
   releasing=0
   selecting=0
   unfocusing=0
   unhovering=0
  }
 }
 HoveredSound {type=SoundBase, default=/Game/Sounds/Fort_UI_Sounds/FrontEnd/New_FrontEnd/DefaultSounds/PressedButtons/FN_UI_Default_Button_Hovered_Cue}
 IsInstantSelection {type=boolean}
 PressedSound {type=SoundBase, default=/Game/Sounds/Fort_UI_Sounds/FrontEnd/New_FrontEnd/DefaultSounds/PressedButtons/FN_UI_Default_Button_Pressed_Cue}
]
widgets {info {inheritedWidgetCount=0, parentClass=/Script/FortniteUI.UIKitModularButton, rootWidgetClass=null, widgetCount=0}, widgets []}
